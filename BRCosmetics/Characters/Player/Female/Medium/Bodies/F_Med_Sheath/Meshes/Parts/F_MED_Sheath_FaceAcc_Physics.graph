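# Physics asset for the Sheath female medium character's face accessory mesh, setting up secondary motion for pigtails, bangs and earrings.
# Capsules on upper spine, clavicles, neck and head form the body volume, joined by constraints with free angular motion.
# Each five-link pigtail hangs off a small base body on the head, with swing limits loosening toward the tip.
# Bangs and earrings constrain directly to the head with locked twist, and earrings swing on one axis only.
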PhysicsAsset /BRCosmetics/Characters/Player/Female/Medium/Bodies/F_Med_Sheath/Meshes/Parts/F_MED_Sheath_FaceAcc_Physics
bodies {
 clavicle_l [{shapeType=Capsule, length=15.9357, radius=5.7, center=<4.6886 -0.0001 0>, rotation=<rot 90 0 0>}]
 clavicle_r [{shapeType=Capsule, length=16.0915, radius=5.7, center=<-5.1432 -0.0004 0>, rotation=<rot 90 0 0>}]
 dyn_bang_l [{shapeType=Capsule, length=3.8366, radius=2.4154, center=<1.6766 0.0001 0>, rotation=<rot 90 0 0>}]
 dyn_bang_r [{shapeType=Capsule, length=4.4785, radius=2.4992, center=<1.9202 -0.0001 0>, rotation=<rot 90 0 0>}]
 dyn_earring_l [{shapeType=Capsule, length=2.2061, radius=1.0151, center=<1.8817 0.0136 0.0005>, rotation=<rot 0 90.5149 -88.282>}]
 dyn_earring_r [{shapeType=Capsule, length=2.2061, radius=1.0151, center=<1.8817 0.0136 -0.0006>, rotation=<rot 0 -89.4851 -88.282>}]
 dyn_pigtail_01_Base_l [{shapeType=Capsule, length=1, radius=1}]
 dyn_pigtail_01_Base_r [{shapeType=Capsule, length=1, radius=1}]
 dyn_pigtail_01_l [{shapeType=Capsule, length=5.8593, radius=4.9078, center=<2 0 0>, rotation=<rot 90 0 0>}]
 dyn_pigtail_01_r [{shapeType=Capsule, length=6.5159, radius=4.1805, center=<2 0 0>, rotation=<rot 90 0 0>}]
 dyn_pigtail_02_l [{shapeType=Capsule, length=6.1585, radius=3.54, center=<2 0 0>, rotation=<rot 90 0 0>}]
 dyn_pigtail_02_r [{shapeType=Capsule, length=6.7729, radius=3.2442, center=<2 0 0>, rotation=<rot 90 0 0>}]
 dyn_pigtail_03_l [{shapeType=Capsule, length=3.145, radius=2.4414, center=<2 0 0>, rotation=<rot 90 0 0>}]
 dyn_pigtail_03_r [{shapeType=Capsule, length=3.0099, radius=2.4405, center=<2 0 0>, rotation=<rot 90 0 0>}]
 dyn_pigtail_04_l [{shapeType=Capsule, length=3.0073, radius=2.4112, center=<2 0 0>, rotation=<rot 90 0 0>}]
 dyn_pigtail_04_r [{shapeType=Capsule, length=2.9557, radius=2.3531, center=<2 0 0>, rotation=<rot 90 0 0>}]
 dyn_pigtail_05_l [{shapeType=Capsule, length=2.5804, radius=1.6229, center=<2 0 0>, rotation=<rot 90 0 0>}]
 dyn_pigtail_05_r [{shapeType=Capsule, length=2.5401, radius=1.6609, center=<2 0 0>, rotation=<rot 90 0 0>}]
 head [{shapeType=Capsule, length=15.0186, radius=16.5152, center=<-0.2669 14.7306 0.0001>, rotation=<rot 0 -180 180>}]
 neck_01 [{shapeType=Capsule, length=60.5001, radius=5.4574, center=<0 1.6569 0>, rotation=<rot 0 -90 0>}]
 neck_02 [{shapeType=Capsule, length=49.5001, radius=4.4574, center=<0 1.1073 0>, rotation=<rot 0 -90 0>}]
 spine_04 [{shapeType=Capsule, length=32.5, radius=14, shapeName=spine_04_capsule, center=<0.4255 6.8799 0>, rotation=<rot 0 -90 0>}]
 spine_05 [{shapeType=Capsule, length=17.746, radius=8.9188, center=<2.2138 1.2357 0>, rotation=<rot 0 -90 0>}]
}
constraints [
 {bone1Name=dyn_pigtail_01_r, bone2Name=dyn_pigtail_01_Base_r, swing1LimitDegrees=10, swing2LimitDegrees=10, twistLimitDegrees=6}
 {bone1Name=dyn_pigtail_02_r, bone2Name=dyn_pigtail_01_r, swing1LimitDegrees=12, swing2LimitDegrees=12, twistLimitDegrees=8}
 {bone1Name=dyn_pigtail_03_r, bone2Name=dyn_pigtail_02_r, swing1LimitDegrees=24, swing2LimitDegrees=24, twistLimitDegrees=8}
 {bone1Name=dyn_pigtail_04_r, bone2Name=dyn_pigtail_03_r, swing1LimitDegrees=24, swing2LimitDegrees=24, twistLimitDegrees=8}
 {bone1Name=dyn_pigtail_05_r, bone2Name=dyn_pigtail_04_r, swing1LimitDegrees=24, swing2LimitDegrees=24, twistLimitDegrees=8}
 {bone1Name=dyn_pigtail_01_l, bone2Name=dyn_pigtail_01_Base_l, swing1LimitDegrees=10, swing2LimitDegrees=10, twistLimitDegrees=6}
 {bone1Name=dyn_pigtail_02_l, bone2Name=dyn_pigtail_01_l, swing1LimitDegrees=12, swing2LimitDegrees=12, twistLimitDegrees=8}
 {bone1Name=dyn_pigtail_03_l, bone2Name=dyn_pigtail_02_l, swing1LimitDegrees=24, swing2LimitDegrees=24, twistLimitDegrees=8}
 {bone1Name=dyn_pigtail_04_l, bone2Name=dyn_pigtail_03_l, swing1LimitDegrees=24, swing2LimitDegrees=24, twistLimitDegrees=8}
 {bone1Name=dyn_pigtail_05_l, bone2Name=dyn_pigtail_04_l, swing1LimitDegrees=24, swing2LimitDegrees=24, twistLimitDegrees=8}
 {
  bone1Name=head
  bone2Name=neck_02
  swing1LimitDegrees=45
  swing1Motion=Free
  swing2LimitDegrees=45
  swing2Motion=Free
  twistLimitDegrees=45
  twistMotion=Free
 }
 {
  bone1Name=clavicle_l
  bone2Name=spine_05
  swing1LimitDegrees=45
  swing1Motion=Free
  swing2LimitDegrees=45
  swing2Motion=Free
  twistLimitDegrees=45
  twistMotion=Free
 }
 {
  bone1Name=clavicle_r
  bone2Name=spine_05
  swing1LimitDegrees=45
  swing1Motion=Free
  swing2LimitDegrees=45
  swing2Motion=Free
  twistLimitDegrees=45
  twistMotion=Free
 }
 {bone1Name=dyn_bang_r, bone2Name=head, swing1LimitDegrees=16, swing2LimitDegrees=4, twistLimitDegrees=4, twistMotion=Locked}
 {bone1Name=dyn_bang_l, bone2Name=head, swing1LimitDegrees=16, swing2LimitDegrees=4, twistLimitDegrees=4, twistMotion=Locked}
 {
  bone1Name=dyn_earring_l
  bone2Name=head
  swing1LimitDegrees=24
  swing2LimitDegrees=45
  swing2Motion=Locked
  twistLimitDegrees=45
  twistMotion=Locked
 }
 {
  bone1Name=dyn_earring_r
  bone2Name=head
  swing1LimitDegrees=24
  swing2LimitDegrees=45
  swing2Motion=Locked
  twistLimitDegrees=45
  twistMotion=Locked
 }
 {
  bone1Name=neck_01
  bone2Name=spine_05
  swing1LimitDegrees=45
  swing1Motion=Free
  swing2LimitDegrees=45
  swing2Motion=Free
  twistLimitDegrees=45
  twistMotion=Free
 }
 {
  bone1Name=neck_02
  bone2Name=neck_01
  swing1LimitDegrees=45
  swing1Motion=Free
  swing2LimitDegrees=45
  swing2Motion=Free
  twistLimitDegrees=45
  twistMotion=Free
 }
 {
  bone1Name=dyn_pigtail_01_Base_r
  bone2Name=head
  swing1LimitDegrees=45
  swing1Motion=Free
  swing2LimitDegrees=45
  swing2Motion=Free
  twistLimitDegrees=45
  twistMotion=Free
 }
 {
  bone1Name=dyn_pigtail_01_Base_l
  bone2Name=head
  swing1LimitDegrees=45
  swing1Motion=Free
  swing2LimitDegrees=45
  swing2Motion=Free
  twistLimitDegrees=45
  twistMotion=Free
 }
 {
  bone1Name=spine_05
  bone2Name=spine_04
  swing1LimitDegrees=45
  swing1Motion=Free
  swing2LimitDegrees=45
  swing2Motion=Free
  twistLimitDegrees=45
  twistMotion=Free
 }
]
