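# Post-attach AnimBP for the Female Medium Sheath face accessory mesh, animating its pigtails.
# It copies the attached parent's pose, then additively rotates both pigtail base bones using HeadForwardAlpha, HeadTwistAlpha and SkydivingAlpha, mostly through remapped ranges.
# A RigidBody node simulates them in head-bone space with damped component motion.
# After physics, range-mapped HeadForwardAlpha offsets adjust lower pigtail bones at more extreme values; the EventGraph is empty.
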AnimBlueprint /BRCosmetics/Characters/Player/Female/Medium/Bodies/F_Med_Sheath/Meshes/Parts/F_MED_Sheath_FaceAcc_AnimBP
target_skeleton=&/F_MED_Sheath_FaceAcc_Skeleton
graphs [
 AnimGraph {nodes [
   n1 = Get HeadForwardAlpha() -> (HeadForwardAlpha:float)
   n2 = "Transform (Modify) Bone - Bone: dyn_pigtail_01_Base_l"(
    ComponentPose=Local To Component(LocalPose=Copy Pose From Mesh() {bUseAttachedParent=true})
    Alpha=@n1
    Translation="0,0,0"
    Rotation="-5,-15,0"
    Scale="1,1,1"
   ) {
    rotationMode=BMM_Additive
    rotationSpace=BCS_BoneSpace
    translationMode=BMM_Additive
    translationSpace=BCS_BoneSpace
    alphaBoolBlend {blendOption=Linear}
    alphaScaleBiasClamp {bMapRange=true, inRange {max=-20, min=-50}, outRange {max=0, min=1}}
    boneToModify {boneName=dyn_pigtail_01_Base_l}
    rotation {pitch=-5, yaw=-15}
   }
   n3 = "Transform (Modify) Bone - Bone: dyn_pigtail_01_Base_r"(ComponentPose=@n2, Alpha=@n1, Translation="0,0,0", Rotation="5,-15,0", Scale="1,1,1") {
    rotationMode=BMM_Additive
    rotationSpace=BCS_BoneSpace
    translationMode=BMM_Additive
    translationSpace=BCS_BoneSpace
    alphaBoolBlend {blendOption=Linear}
    alphaScaleBiasClamp {bMapRange=true, inRange {max=-20, min=-50}, outRange {max=0, min=1}}
    boneToModify {boneName=dyn_pigtail_01_Base_r}
    rotation {pitch=5, yaw=-15}
   }
   n4 = Get HeadTwistAlpha() -> (HeadTwistAlpha:float)
   n5 = "Transform (Modify) Bone - Bone: dyn_pigtail_01_Base_l"(ComponentPose=@n3, Alpha=@n4, Translation="0,0,0", Rotation="-20,0,0", Scale="1,1,1") {
    rotationMode=BMM_Additive
    rotationSpace=BCS_BoneSpace
    translationMode=BMM_Additive
    translationSpace=BCS_BoneSpace
    alphaBoolBlend {blendOption=Linear}
    alphaScaleBiasClamp {bMapRange=true, inRange {max=0, min=-30}, outRange {max=0, min=1}}
    boneToModify {boneName=dyn_pigtail_01_Base_l}
    rotation {pitch=-20}
   }
   n6 = "Transform (Modify) Bone - Bone: dyn_pigtail_01_Base_r"(ComponentPose=@n5, Alpha=@n4, Translation="0,0,0", Rotation="25,0,0", Scale="1,1,1") {
    rotationMode=BMM_Additive
    rotationSpace=BCS_BoneSpace
    translationMode=BMM_Additive
    translationSpace=BCS_BoneSpace
    alphaBoolBlend {blendOption=Linear}
    alphaScaleBiasClamp {bMapRange=true, inRange {max=40}}
    boneToModify {boneName=dyn_pigtail_01_Base_r}
    rotation {pitch=25}
   }
   n7 = Get SkydivingAlpha() -> (SkydivingAlpha:float)
   n8 = "Transform (Modify) Bone - Bone: dyn_pigtail_01_Base_l"(ComponentPose=@n6, Alpha=@n7, Translation="0,0,0", Rotation="-20,-40,0", Scale="1,1,1") {
    rotationMode=BMM_Additive
    rotationSpace=BCS_BoneSpace
    translationMode=BMM_Additive
    translationSpace=BCS_BoneSpace
    alphaBoolBlend {blendOption=Linear}
    boneToModify {boneName=dyn_pigtail_01_Base_l}
    rotation {pitch=-20, yaw=-40}
   }
   n9 = "Transform (Modify) Bone - Bone: dyn_pigtail_01_Base_r"(ComponentPose=@n8, Alpha=@n7, Translation="0,0,0", Rotation="20,-40,0", Scale="1,1,1") {
    rotationMode=BMM_Additive
    rotationSpace=BCS_BoneSpace
    translationMode=BMM_Additive
    translationSpace=BCS_BoneSpace
    alphaBoolBlend {blendOption=Linear}
    boneToModify {boneName=dyn_pigtail_01_Base_r}
    rotation {pitch=20, yaw=-40}
   }
   n10 = RigidBody(ComponentPose=@n9) {
    bClampLinearTranslationLimitToRefPose=true
    bOverrideWorldGravity=true
    overridePhysicsAsset=&/F_MED_Sheath_FaceAcc_Physics
    simulationSpace=BaseBoneSpace
    alphaBoolBlend {blendOption=Linear}
    baseBoneRef {boneName=head}
    componentAppliedLinearAccClamp=<4000 4000 5000>
    componentLinearAccScale=<0.3 0.3 0.63>
    componentLinearVelScale=<0.3 0.3 0.255>
    overrideWorldGravity {z=-980}
    simSpaceSettings {worldAlpha=0.2}
   }
   n11 = "Transform (Modify) Bone - Bone: dyn_pigtail_02_l"(ComponentPose=@n10, Alpha=@n1, Translation="-0.196,-2.284,0.000", Rotation="0.000,-11.728,-0.000", Scale="1,1,1") {
    rotationMode=BMM_Additive
    rotationSpace=BCS_BoneSpace
    translationMode=BMM_Additive
    translationSpace=BCS_BoneSpace
    alphaBoolBlend {blendOption=Linear}
    alphaScaleBiasClamp {bMapRange=true, inRange {max=-50, min=-90}, outRange {max=0, min=1}}
    boneToModify {boneName=dyn_pigtail_02_l}
    rotation=<rot 0 -11.728 0>
    translation=<-0.196 -2.284>
   }
   n12 = "Transform (Modify) Bone - Bone: dyn_pigtail_03_l"(ComponentPose=@n11, Alpha=@n1, Translation="0,0,0", Rotation="0.001,18.271,0.010", Scale="1,1,1") {
    rotationMode=BMM_Additive
    rotationSpace=BCS_BoneSpace
    translationMode=BMM_Additive
    translationSpace=BCS_BoneSpace
    alphaBoolBlend {blendOption=Linear}
    alphaScaleBiasClamp {bMapRange=true, inRange {max=-50, min=-90}, outRange {max=0, min=1}}
    boneToModify {boneName=dyn_pigtail_03_l}
    rotation=<rot 0.001 18.271 0.01>
   }
   n13 = "Transform (Modify) Bone - Bone: dyn_pigtail_02_r"(ComponentPose=@n12, Alpha=@n1, Translation="-0.618,-3.564,1.448", Rotation="0,0,0", Scale="1,1,1") {
    rotationMode=BMM_Additive
    rotationSpace=BCS_BoneSpace
    translationMode=BMM_Additive
    translationSpace=BCS_BoneSpace
    alphaBoolBlend {blendOption=Linear}
    alphaScaleBiasClamp {bMapRange=true, inRange {max=-50, min=-90}, outRange {max=0, min=1}}
    boneToModify {boneName=dyn_pigtail_02_r}
    translation=<-0.618 -3.564 1.448>
   }
   n14 = "Transform (Modify) Bone - Bone: dyn_pigtail_03_r"(ComponentPose=@n13, Alpha=@n1, Translation="0,0,0", Rotation="-47.099,-0.069,-0.040", Scale="1,1,1") {
    rotationMode=BMM_Additive
    rotationSpace=BCS_BoneSpace
    translationMode=BMM_Additive
    translationSpace=BCS_BoneSpace
    alphaBoolBlend {blendOption=Linear}
    alphaScaleBiasClamp {bMapRange=true, inRange {max=-50, min=-90}, outRange {max=0, min=1}}
    boneToModify {boneName=dyn_pigtail_03_r}
    rotation=<rot -47.099 -0.069 -0.04>
   }
   Output Pose(Result=Component To Local(ComponentPose=@n14)) {layerGroup=None}
  ]}
 EventGraph {links [], nodes []}
]
variables []
